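; Lootable safe prop Blueprint that adds cosmetic feedback when the container is looted.
; On loot, it hides the chest smoke sheet and its children, stops the ambient sound, and spawns the normal chest burst Niagara effect at the static mesh's location and rotation.
; The construction script only calls the parent, and BeginPlay, Tick and overlap events are disabled stubs.
; Despite the name, no physics logic appears in this graph.
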(event EventBeginPlay
  (|CallParentFunction) ; This node is disabled and will not be called.\nDrag off pins to build functionality.
)

(event Collision|EventActorBeginOverlap
  (params (OtherActor :type Actor%20Object%20Reference))
  (|CallParentFunction OtherActor) ; This node is disabled and will not be called.\nDrag off pins to build functionality.
)

(event EventTick
  (params (DeltaSeconds :type Float%20%28single-precision%29))
  (|CallParentFunction DeltaSeconds) ; This node is disabled and will not be called.\nDrag off pins to build functionality.
)

(event EventOnLoot
  (Development|SetHiddeninGame (Variables|Default|GetChestSmokeSheet) true true)
  (Components|Activation|Deactivate (Variables|Default|GetChestAmbientSound))
  (Niagara|SpawnSystematLocation "/Game/Effects/Fort_Effects/World/NewChests/NS_ChestBurst_Normal.NS_ChestBurst_Normal" (.location (Transformation|GetWorldTransform (Variables|Debug|GetStaticMeshComponent))) (.rotation (Transformation|GetWorldTransform (Variables|Debug|GetStaticMeshComponent)))))

(fn ConstructionScript
  (|CallParentFunction))
